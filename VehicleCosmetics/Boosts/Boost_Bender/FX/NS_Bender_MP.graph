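# NS_Bender_MP is a Niagara system for the Bender vehicle boost cosmetic, with no emitters listed.
# System scripts use System State and a spawn-time Set Variables assignment, with skeletal mesh data interface and Add Vector to Position dependencies.
# User parameters expose paint, color variant, game mode, frontend preview, port count, boost position and car offset, though their wiring is not shown.
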NiagaraSystem /VehicleCosmetics/Boosts/Boost_Bender/FX/NS_Bender_MP
dependencies {
 usedDataInterfaces [&NiagaraDataInterfaceSkeletalMesh]
 usedDynamicInputs [&AddVectorToPosition]
 usedModules [
  &self:SystemSpawnScript.SystemScriptSource.SystemScriptGraph.NiagaraNodeAssignment_0.SetVariables_#1
  &SystemState
 ]
}
summary {
 systemName=NS_Bender_MP
 emitters []
 userVariables [
  User.bIsFrontendPreview {type=NiagaraBool}
  User.bIsPainted {type=NiagaraBool}
  User.ColorVariantID {type=NiagaraInt32}
  User.DTCarOffset {type=Vector3f, description="", defaultValue {struct=Vector3f, value=<0 0 0>}}
  User.GameModeID {type=NiagaraInt32}
  User.PaintedColor {type=LinearColor, description="", defaultValue {struct=LinearColor, value=<rgba 1 1 1 1>}}
  User.PaintedScalar {type=NiagaraFloat, defaultValue=1, description=""}
  User.PortCount {type=NiagaraInt32, defaultValue=1}
  User.SingleBoostPos {type=NiagaraPosition, description="", defaultValue {}}
 ]
}
