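# Niagara system for the Bender vehicle boost cosmetic, a rear exhaust pairing orange and blue streams.
# Quality-scaled ribbons trail backward alongside short-lived mesh bursts, growing sprites with small curl-noise particles, and larger core sprites.
# Spawn groups merge variants into shared emitters, selecting per-group size, offset and forces and setting renderer visibility tags.
# When bIsPainted is set, PolarColor generally replaces orange and PaintedColor blue, with core sprites masking between both.
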NiagaraSystem /VehicleCosmetics/Boosts/Boost_Bender/FX/NS_Boost_Bender
dependencies {
 usedDataInterfaces [
  &NiagaraDataInterfaceVectorCurve
  &NiagaraDataInterfaceCurve
  &NiagaraDataInterfaceSpriteRendererInfo
  &NiagaraDataInterfaceArrayColor
  &NiagaraDataInterfaceVector2DCurve
  &NiagaraDataInterfaceSkeletalMesh
 ]
 usedDynamicInputs []
 usedModules [
  &self:SystemSpawnScript.SystemScriptSource.SystemScriptGraph.NiagaraNodeAssignment_0.SetVariables_#1
  &SystemState
 ]
}
emitters {
 NE_VC_SingleBoost_Ribbon_Parent001 {
  emitterUpdateScript [
   &EmitterState()
   &VC_FilteredSocketCount()
   &VC_RateQualityHandler(Epic Spawn Rate=120, High Spawn Rate=90, Low Spawn Rate=30, Medium Spawn Rate=45, Use Custom Rate?=true)
   &VC_SpawnRibbon_v1(SpawnRate=$Emitter.Spawn Rate)
  ]
  particleSpawnScript [
   &VC_PaintedVariantsOverride() {enabled=false}
   &VC_InitializeRibbon_v1(
    Color=&VC_DynamicScaleColor(
     Color to Scale:LinearColor=&MakeCustomLinearColorFromBool(Bool=$User.bIsPainted, True Color=$User.PolarColor, False Color=<rgba 2 0.4 0.06 1>)
     RGB Scale:Vector3f=<1 1 1>
     Alpha Scale:NiagaraFloat=1
    )
    Custom Offset=true
    Lifetime=0.4
    Local Vector Offset=&AddVectorToPosition(Position=<-250 -20 30>, Vector=<0 0 0>)
    Offset Amount=1
    Ribbon Width:NiagaraFloat=&Multiply_Float(A=45, B=1)
   )
   &AddVelocity(Velocity=<-1000 0 0>, Velocity Speed Scale=2) {module=AddVelocity001}
  ]
  particleUpdateScript [
   &ParticleState()
   /SetVariables(Particles.BlueRColor:LinearColor=&VC_DynamicScaleColor(
     Color to Scale:LinearColor=&MakeCustomLinearColorFromBool(Bool=$User.bIsPainted, True Color=$User.PaintedColor, False Color=<rgba 0 0.7557 5 1>)
     RGB Scale:Vector3f=<1 1 1>
     Alpha Scale:NiagaraFloat=1
    )) {isSetParameters=true}
   &Color(Color=&VC_DynamicScaleColor(
     Color to Scale:LinearColor=&MakeCustomLinearColorFromBool(Bool=$User.bIsPainted, True Color=$User.PolarColor, False Color=<rgba 5 0.8763 0 1>)
     RGB Scale:Vector3f=<1 1 1>
     Alpha Scale:NiagaraFloat=1
    )) {enabled=false}
   &ScaleColor(
    Scale Alpha=&FloatFromCurve(FloatCurve=DataInterface() {propertyValues {Curve {keys [
         <at 0 1 RCIM_Cubic>
         {arriveTangent=0.0119, interpMode=RCIM_Cubic, leaveTangent=0.0119, tangentMode=RCTM_User, time=0.2, value=1}
         <at 0.8 1>
         <at 1 0>
        ]}}})
    Scale RGB=&VectorFromCurve(VectorCurve=DataInterface() {propertyValues {
       XCurve {keys [<at 0 1 RCIM_Cubic>,<at 1 0>]}
       YCurve {keys [<at 0 1 RCIM_Cubic>,<at 1 0>]}
       ZCurve {keys [<at 0 1 RCIM_Cubic>,<at 1 0 RCIM_Cubic>]}
      }})
    ScaleA=true
    ScaleRGB=true
   )
   &ScaleRibbonWidth(Ribbon Width Scale=&FloatFromCurve(FloatCurve=DataInterface() {propertyValues {Curve {keys [<at 0 1>,<at 0.5 0.5>]}}}))
   &SolveForcesAndVelocity(Clamp Velocity=false, Limit Acceleration=false)
   &VC_UpdateRibbon_v1()
  ]
  rendererClasses [&NiagaraRibbonRendererProperties]
  renderers [
   {rendererClass=&NiagaraRibbonRendererProperties}
   {rendererClass=&NiagaraRibbonRendererProperties, rendererIndex=1}
  ]
 }
 NE_VC_SingleBoost_Sprite_Parent001 {
  emitterUpdateScript [
   &EmitterState()
   &SpawnRate(SpawnRate=5, Use Spawn Probability=false)
   &SpawnRate(SpawnGroup=1, SpawnRate=5, Use Spawn Probability=false) {module=SpawnRate001}
  ]
  particleSpawnScript [
   &VC_PaintedVariantsOverride() {enabled=false}
   &InitializeParticle(
    Color=&VC_DynamicScaleColor(
     Color to Scale:LinearColor=&MakeCustomLinearColorFromBool(Bool=$User.bIsPainted, True Color=$User.PolarColor, False Color=<rgba 1 0.2 0.03 1>)
     RGB Scale:Vector3f=<1 1 1>
     Alpha Scale:NiagaraFloat=1
    )
    Lifetime=&RandomRangeFloat(Minimum=0.125, Maximum=0.2)
    Mesh Scale=&Multiply_VectorByFloat(Vector=&RandomRangeVector(Minimum=<6.2 1 1>, Maximum=<6.2 1.25 1>), Float=1.2)
    Mesh Scale Mode=::NewEnumerator4
    Position Offset=&AddVectorToPosition(Position=<-200 0 25>, Vector=<0 0 0>)
    Position Offset Coordinate Space=::Local
    Sprite Rotation Mode=::NewEnumerator2
    UsePositionOffset=true
    Write Material Random=true
   )
   &AddVelocity(Velocity=&RandomRangeVector(Minimum=<-15 0 0>, Maximum=<-35 1 1>))
   &InitialMeshOrientation(
    Mesh Orientation Mode=::NewEnumerator2
    Rotation=&RandomRangeVector(Minimum=<-10 0 0>, Maximum=<10 0 0>)
    Use Rotation Vector=true
   )
   &ShapeLocation(MaskSpawnGroup=false, Override Seed=false, Sphere Radius=0, UseOwnerScale=true)
  ]
  particleUpdateScript [
   &ParticleState()
   /SetVariables(
    Particles.BlueMColor:LinearColor=&VC_DynamicScaleColor(
     Color to Scale:LinearColor=&MakeCustomLinearColorFromBool(Bool=$User.bIsPainted, True Color=$User.PaintedColor, False Color=<rgba 0.03 0.5 1 1>)
     RGB Scale:Vector3f=<1 1 1>
     Alpha Scale:NiagaraFloat=1
    )
    Particles.MTagVis:NiagaraInt32=&SpawnGroup()
   ) {isSetParameters=true}
   &ScaleColor(Scale RGB=<3 0 0>, ScaleA=true, ScaleRGB=true) {enabled=false}
   &ScaleMeshSize(Scale Factor=&VectorFromCurve(VectorCurve=DataInterface() {propertyValues {XCurve {keys [<at 0 1>,<at 1 1.2>]}, YCurve {keys [<at 0 1>,<at 1 1.1>]}, ZCurve {keys [<at 0 1>,<at 1 1.1>]}}}))
   &DynamicMaterialParameters(
    Index 0 Param 1=1
    Index 0 Param 4=10
    Module.Index0Float0Write=true
    Module.Index0Float1Write=true
    Module.Index0Float2Write=true
    Module.Index0Float3Write=true
   )
   &SolveForcesAndVelocity(Clamp Velocity=false, Limit Acceleration=false)
  ]
  rendererClasses [&NiagaraSpriteRendererProperties, &NiagaraMeshRendererProperties]
  renderers [
   {rendererClass=&NiagaraSpriteRendererProperties}
   {rendererClass=&NiagaraMeshRendererProperties, rendererIndex=1}
   {rendererClass=&NiagaraMeshRendererProperties, rendererIndex=2}
  ]
 }
 NE_VC_SingleBoost_Sprite_Parent002 {
  emitterUpdateScript [
   &EmitterState()
   &SpawnRate(SpawnRate=35, Use Spawn Probability=false)
   &SpawnRate(SpawnGroup=1, SpawnRate=35, Use Spawn Probability=false) {module=SpawnRate001}
   &SpawnRate(SpawnGroup=2, SpawnRate=25, Use Spawn Probability=false) {module=SpawnRate002}
   &SpawnRate(SpawnGroup=3, SpawnRate=25, Use Spawn Probability=false) {module=SpawnRate003}
  ]
  particleSpawnScript [
   &VC_PaintedVariantsOverride(
    R3 DefaultColor:LinearColor=&MakeCustomLinearColorFromBool(Bool=$User.bIsPainted, True Color=$User.PaintedColor, False Color=<rgba 0 5.2631 20 50>)
    Use R3 Painted Overrides=true
   ) {enabled=false}
   &InitializeParticle(
    Color=&VC_DynamicScaleColor(
     Color to Scale:LinearColor=&MakeCustomLinearColorFromBool(Bool=$User.bIsPainted, True Color=$User.PolarColor, False Color=<rgba 1.5 0.1224 0 1>)
     RGB Scale:Vector3f=<1 1 1>
     Alpha Scale:NiagaraFloat=1
    )
    Lifetime=/VehicleCosmetics/FX/Niagara/Modules/MergeEmitters/VC_FloatBySpawnGroup(
     SpawnGroup or ExecIndex:NiagaraInt32=0
     CustomInt:NiagaraInt32=0
     MaximumSpawnGroups:NiagaraInt32=4
     Float SG 0:NiagaraFloat=&RandomRangeFloat(Minimum=0.35, Maximum=0.55)
     Float SG 1:NiagaraFloat=&RandomRangeFloat(Minimum=0.35, Maximum=0.55)
     Float SG 2:NiagaraFloat=1
     Float SG 3:NiagaraFloat=1
     Float SG 4:NiagaraFloat=0
     Float SG 5:NiagaraFloat=0
    )
    Position Offset=/VehicleCosmetics/FX/Niagara/Modules/MergeEmitters/VC_VectorBySpawnGroup(
     SpawnGroup or ExecIndex:NiagaraInt32=0
     CustomInt:NiagaraInt32=0
     MaximumSpawnGroups:NiagaraInt32=4
     Vector Group 0:Vector3f=<-250 -15 25>
     Vector Group 1:Vector3f=<-250 15 25>
     Vector Group 2:Vector3f=<-250 -15 25>
     Vector Group 3:Vector3f=<-250 25 25>
     Vector Group 4:Vector3f=<0 0 0>
     Vector Group 5:Vector3f=<0 0 0>
    )
    Position Offset Coordinate Space=::Local
    Sprite Rotation Mode=::NewEnumerator2
    Sprite Size Mode=::NewEnumerator3
    Uniform Sprite Size=&VC_FloatBySpawnGroup(
     SpawnGroup or ExecIndex:NiagaraInt32=0
     CustomInt:NiagaraInt32=0
     MaximumSpawnGroups:NiagaraInt32=4
     Float SG 0:NiagaraFloat=&RandomRangeFloat(Minimum=55, Maximum=65)
     Float SG 1:NiagaraFloat=&RandomRangeFloat(Minimum=55, Maximum=65)
     Float SG 2:NiagaraFloat=&RandomRangeFloat(Minimum=1, Maximum=3)
     Float SG 3:NiagaraFloat=&RandomRangeFloat(Minimum=1, Maximum=3)
     Float SG 4:NiagaraFloat=0
     Float SG 5:NiagaraFloat=0
    )
    UsePositionOffset=true
    Write Material Random=true
   )
   &AddVelocity(Velocity=&VC_VectorBySpawnGroup(
     SpawnGroup or ExecIndex:NiagaraInt32=0
     CustomInt:NiagaraInt32=0
     MaximumSpawnGroups:NiagaraInt32=4
     Vector Group 0:Vector3f=<-1000 0 0>
     Vector Group 1:Vector3f=<-1000 0 0>
     Vector Group 2:Vector3f=&RandomRangeVector(Minimum=<-1000 50 50>, Maximum=<-1000 -50 -50>)
     Vector Group 3:Vector3f=&RandomRangeVector(Minimum=<-1000 50 50>, Maximum=<-1000 -50 -50>)
     Vector Group 4:Vector3f=<0 0 0>
     Vector Group 5:Vector3f=<0 0 0>
    ))
   &ShapeLocation(
    MaskSpawnGroup=false
    Override Seed=false
    Sphere Radius=&VC_FloatBySpawnGroup(
     SpawnGroup or ExecIndex:NiagaraInt32=0
     CustomInt:NiagaraInt32=0
     MaximumSpawnGroups:NiagaraInt32=4
     Float SG 0:NiagaraFloat=0
     Float SG 1:NiagaraFloat=0
     Float SG 2:NiagaraFloat=&RandomRangeFloat(Minimum=5, Maximum=15)
     Float SG 3:NiagaraFloat=&RandomRangeFloat(Minimum=5, Maximum=15)
     Float SG 4:NiagaraFloat=0
     Float SG 5:NiagaraFloat=0
    )
    UseOwnerScale=true
   )
  ]
  particleUpdateScript [
   &ParticleState()
   /SetVariables(
    Particles.BEMColor:LinearColor=&VC_DynamicScaleColor(
     Color to Scale:LinearColor=&MakeCustomLinearColorFromBool(Bool=$User.bIsPainted, True Color=$User.PaintedColor, False Color=<rgba 0 13.1577 50 50>)
     RGB Scale:Vector3f=<1 1 1>
     Alpha Scale:NiagaraFloat=1
    )
    Particles.BFColor:LinearColor=&VC_DynamicScaleColor(
     Color to Scale:LinearColor=&MakeCustomLinearColorFromBool(Bool=$User.bIsPainted, True Color=$User.PaintedColor, False Color=<rgba 0.05 0.3 1 1>)
     RGB Scale:Vector3f=&VectorFromCurve(VectorCurve=DataInterface() {propertyValues {XCurve {keys [<at 0 2>,<at 1 5>]}, YCurve {keys [<at 0 2>,<at 1 5>]}, ZCurve {keys [<at 0 2>,<at 1 5>]}}})
     Alpha Scale:NiagaraFloat=&FloatFromCurve(FloatCurve=DataInterface() {propertyValues {Curve {keys [<at 0.3 5>,<at 0.5 1>]}}})
    )
    Particles.REmColor:LinearColor=&VC_DynamicScaleColor(
     Color to Scale:LinearColor=&MakeCustomLinearColorFromBool(Bool=$User.bIsPainted, True Color=$User.PolarColor, False Color=<rgba 50 4.0816 0 1>)
     RGB Scale:Vector3f=<1 1 1>
     Alpha Scale:NiagaraFloat=&FloatFromCurve(FloatCurve=DataInterface() {propertyValues {Curve {keys [<at 0 1>,<at 1 0>]}}})
    )
    Particles.SpTagVis:NiagaraInt32=&SpawnGroup()
   ) {isSetParameters=true}
   &ScaleColor(
    Scale Alpha=&FloatFromCurve(FloatCurve=DataInterface() {propertyValues {Curve {keys [<at 0.3 5>,<at 0.5 1>]}}})
    Scale RGB=&VectorFromCurve(VectorCurve=DataInterface() {propertyValues {XCurve {keys [<at 0 2>,<at 1 5>]}, YCurve {keys [<at 0 2>,<at 1 5>]}, ZCurve {keys [<at 0 2>,<at 1 5>]}}})
    ScaleA=true
    ScaleRGB=true
   )
   &ScaleSpriteSize(Uniform Curve Sprite Scale=DataInterface() {propertyValues {Curve {keys [<at 0 0.5 RCIM_Cubic>,<at 1 5 RCIM_Cubic>]}}})
   &SubUVAnimation(
    Output Alternate Renderer Binding=true
    "Sprite Renderer "=DataInterface() {propertyValues {SpriteRenderer=&self:NE_VC_SingleBoost_Sprite_Parent002.NiagaraSpriteRendererProperties_0}}
    UseEndFrame=false
    UseStartFrame=false
   )
   &SubUVAnimation(
    Output Alternate Renderer Binding=true
    "Sprite Renderer "=DataInterface() {propertyValues {SpriteRenderer=&self:NE_VC_SingleBoost_Sprite_Parent002.NiagaraSpriteRendererProperties_1}}
    UseEndFrame=false
    UseStartFrame=false
   ) {enabled=false, module=SubUVAnimation001}
   &CurlNoiseForce(
    Bias Noise=false
    Noise Frequency=&VC_FloatBySpawnGroup(
     SpawnGroup or ExecIndex:NiagaraInt32=0
     CustomInt:NiagaraInt32=0
     MaximumSpawnGroups:NiagaraInt32=4
     Float SG 0:NiagaraFloat=0
     Float SG 1:NiagaraFloat=0
     Float SG 2:NiagaraFloat=1000
     Float SG 3:NiagaraFloat=1000
     Float SG 4:NiagaraFloat=0
     Float SG 5:NiagaraFloat=0
    )
    Noise Strength=&VC_FloatBySpawnGroup(
     SpawnGroup or ExecIndex:NiagaraInt32=0
     CustomInt:NiagaraInt32=0
     MaximumSpawnGroups:NiagaraInt32=4
     Float SG 0:NiagaraFloat=0
     Float SG 1:NiagaraFloat=0
     Float SG 2:NiagaraFloat=250
     Float SG 3:NiagaraFloat=250
     Float SG 4:NiagaraFloat=0
     Float SG 5:NiagaraFloat=0
    )
    Pan Noise=false
   )
   &Drag(
    Drag=&VC_FloatBySpawnGroup(
     SpawnGroup or ExecIndex:NiagaraInt32=0
     CustomInt:NiagaraInt32=0
     MaximumSpawnGroups:NiagaraInt32=4
     Float SG 0:NiagaraFloat=0
     Float SG 1:NiagaraFloat=0
     Float SG 2:NiagaraFloat=&FloatFromCurve(FloatCurve=DataInterface() {propertyValues {Curve {keys [<at 0.75 0>,<at 1 1>]}}})
     Float SG 3:NiagaraFloat=0
     Float SG 4:NiagaraFloat=0
     Float SG 5:NiagaraFloat=0
    )
    Rotational Drag:NiagaraFloat=&FloatFromCurve(FloatCurve=DataInterface() {propertyValues {Curve {keys [<at 0.75 0>,<at 1 1>]}}})
    Use Linear Drag=true
    Use Rotational Drag=true
   )
   &SolveForcesAndVelocity(Clamp Velocity=false, Limit Acceleration=false)
  ]
  rendererClasses [&NiagaraSpriteRendererProperties]
  renderers [
   {rendererClass=&NiagaraSpriteRendererProperties}
   {rendererClass=&NiagaraSpriteRendererProperties, rendererIndex=1}
   {rendererClass=&NiagaraSpriteRendererProperties, rendererIndex=2}
   {rendererClass=&NiagaraSpriteRendererProperties, rendererIndex=3}
  ]
 }
 NE_VC_SingleBoost_Sprite_Parent004 {
  emitterUpdateScript [
   &EmitterState()
   &SpawnRate(SpawnRate=30, Use Spawn Probability=false)
   &SpawnRate(SpawnGroup=1, SpawnRate=32, Use Spawn Probability=false) {module=SpawnRate001}
  ]
  particleSpawnScript [
   &VC_PaintedVariantsOverride() {enabled=false}
   &InitializeParticle(
    Color=&VC_DynamicScaleColor(
     Color to Scale:LinearColor=&MakeCustomLinearColorFromBool(
      Bool=$User.bIsPainted
      True Color=&MaskLinearColor(
       Mask Source=1
       Mask Parameter=0
       Mask=&RandomRangeInt(Minimum=0, Maximum=1)
       Masked Linear Color=$User.PolarColor
       Passthrough Linear Color=$User.PaintedColor
      )
      False Color=&SelectLinearColorFromArray(
       Array Sampling Mode=::NewEnumerator0
       Direct Array Index=?
       Color Selection Array=DataInterface() {propertyValues {ColorData [<rgba 0.05 0.3 1 1>, <rgba 1 0.1 0.02 1>]}}
       Random Seed=0
      )
     )
     RGB Scale:Vector3f=<1 1 1>
     Alpha Scale:NiagaraFloat=1
    )
    Lifetime=&RandomRangeFloat(Minimum=0.35, Maximum=0.5)
    Position Mode=::NewEnumerator1
    Position Offset=&AddVectorToPosition(Position=<-250 0 25>, Vector=<0 0 0>)
    Position Offset Coordinate Space=::Local
    Sprite Rotation Mode=::NewEnumerator2
    Sprite Size Mode=::NewEnumerator3
    Uniform Sprite Size=&VC_FloatBySpawnGroup(
     SpawnGroup or ExecIndex:NiagaraInt32=0
     CustomInt:NiagaraInt32=0
     MaximumSpawnGroups:NiagaraInt32=2
     Float SG 0:NiagaraFloat=180
     Float SG 1:NiagaraFloat=195
     Float SG 2:NiagaraFloat=0
     Float SG 3:NiagaraFloat=0
     Float SG 4:NiagaraFloat=0
     Float SG 5:NiagaraFloat=0
    )
    UsePositionOffset=true
    Write Material Random=true
   )
   &AddVelocity(Velocity=<-1500 0 0>) {enabled=false}
   &ShapeLocation(MaskSpawnGroup=false, Override Seed=false, Sphere Radius=0.05, Sphere Surface Distribution=1, UseOwnerScale=true)
   &CameraOffset(Camera Offset Amount=&VC_FloatBySpawnGroup(
     SpawnGroup or ExecIndex:NiagaraInt32=0
     CustomInt:NiagaraInt32=0
     MaximumSpawnGroups:NiagaraInt32=2
     Float SG 0:NiagaraFloat=0
     Float SG 1:NiagaraFloat=8
     Float SG 2:NiagaraFloat=0
     Float SG 3:NiagaraFloat=0
     Float SG 4:NiagaraFloat=0
     Float SG 5:NiagaraFloat=0
    ))
  ]
  particleUpdateScript [
   &ParticleState()
   /SetVariables(
    Particles.BCoreColor:LinearColor=&VC_DynamicScaleColor(
     Color to Scale:LinearColor=&MakeCustomLinearColorFromBool(
      Bool=$User.bIsPainted
      True Color=&MaskLinearColor(
       Mask Source=1
       Mask Parameter=&RandomRangeInt(Minimum=0, Maximum=1)
       Mask=0
       Masked Linear Color=$User.PolarColor
       Passthrough Linear Color=$User.PaintedColor
      )
      False Color=&SelectLinearColorFromArray(
       Array Sampling Mode=::NewEnumerator0
       Direct Array Index=&ModuloInt(A=&SpawnGroup(), B=2)
       Color Selection Array=DataInterface() {propertyValues {ColorData [<rgba 0.0294 0.1766 0.5885 1>, <rgba 0.4688 0.0469 0.0094 1>]}}
       Random Seed=0
      )
     )
     RGB Scale:Vector3f=<1 1 1>
     Alpha Scale:NiagaraFloat=1
    )
    Particles.FlamesSize:Vector2f=<145 145>
    Particles.ROrbTagVis:NiagaraInt32=&SpawnGroup()
   ) {isSetParameters=true}
   &Color(Color=&VC_DynamicScaleColor(
     Color to Scale:LinearColor=&MakeCustomLinearColorFromBool(Bool=$User.bIsPainted, True Color=$User.PolarColor, False Color=<rgba 1 0.1 0.02 1>)
     RGB Scale:Vector3f=<1 1 1>
     Alpha Scale:NiagaraFloat=1
    )) {enabled=false}
   &ScaleSpriteSize(
    Non-Uniform Curve Sprite Scale=DataInterface() {propertyValues {XCurve {keys [<at 0 0 RCIM_Cubic>,<at 1 1 RCIM_Cubic>]}, YCurve {keys [<at 0 0.5 RCIM_Cubic>,<at 1 1 RCIM_Cubic>]}}}
    Scale Sprite Size Mode=::NewEnumerator3
   )
   &ScaleColor(
    Scale Alpha=&FloatFromCurve(FloatCurve=DataInterface() {propertyValues {Curve {keys [
         <at 0 0.5 RCIM_Cubic>
         {arriveTangent=0.0032, interpMode=RCIM_Cubic, leaveTangent=0.0032, tangentMode=RCTM_User, time=0.5, value=0.5}
         {arriveTangent=-0.5, interpMode=RCIM_Cubic, leaveTangent=-0.5, time=0.8, value=0.25}
         <at 1.5 0 RCIM_Cubic>
        ]}}})
    Scale RGB=&VectorFromCurve(VectorCurve=DataInterface() {propertyValues {
       XCurve {keys [<at 0 2 RCIM_Cubic>,<at 1 0.1>]}
       YCurve {keys [<at 0 2 RCIM_Cubic>,<at 1 0.1>]}
       ZCurve {keys [<at 0 2 RCIM_Cubic>,<at 1 0.1>]}
      }})
    ScaleA=true
    ScaleRGB=true
   ) {enabled=false}
   &SubUVAnimation(
    Output Alternate Renderer Binding=true
    "Sprite Renderer "=DataInterface() {propertyValues {SpriteRenderer=&self:NE_VC_SingleBoost_Sprite_Parent004_0.NiagaraSpriteRendererProperties_0}}
    UseEndFrame=false
    UseStartFrame=false
   )
   &SolveForcesAndVelocity(Clamp Velocity=false, Limit Acceleration=false)
  ]
  rendererClasses [&NiagaraSpriteRendererProperties]
  renderers [
   {rendererClass=&NiagaraSpriteRendererProperties}
   {rendererClass=&NiagaraSpriteRendererProperties, rendererIndex=1}
  ]
 }
}
summary {
 systemName=NS_Boost_Bender
 emitters [
  {
   emitterName=NE_VC_SingleBoost_Sprite_Parent001
   rendererClasses [&NiagaraSpriteRendererProperties, &NiagaraMeshRendererProperties]
  }
  {emitterName=NE_VC_SingleBoost_Ribbon_Parent001, rendererClasses [&NiagaraRibbonRendererProperties]}
  {emitterName=NE_VC_SingleBoost_Sprite_Parent002, rendererClasses [&NiagaraSpriteRendererProperties]}
  {emitterName=NE_VC_SingleBoost_Sprite_Parent004, rendererClasses [&NiagaraSpriteRendererProperties]}
 ]
 userVariables [
  User.bIsFrontendPreview {type=NiagaraBool}
  User.bIsPainted {type=NiagaraBool}
  User.ColorVariantID {type=NiagaraInt32}
  User.PaintedColor {type=LinearColor, description="", defaultValue {struct=LinearColor, value=<rgba 1 1 1 1>}}
  User.PaintedScalar {type=NiagaraFloat, defaultValue=1, description=""}
  User.PolarColor {type=LinearColor, description="", defaultValue {struct=LinearColor, value=<rgba 1 1 1 1>}}
  User.PortCount {type=NiagaraInt32, defaultValue=1}
 ]
}
