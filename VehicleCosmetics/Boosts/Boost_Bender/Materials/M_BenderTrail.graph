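# Unlit translucent Niagara material for the Bender vehicle boost trail, usable on ribbons, sprites and mesh particles.
# Two panning flame-mask samples on UV1 combine with UV0 edge and length gradients to form saturated masks.
# These masks blend square-rooted particle color with a lightness-shifted, brightened version for emissive, and also drive depth-faded opacity.
# A WPO vector parameter is transformed from local space.
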Material /VehicleCosmetics/Boosts/Boost_Bender/Materials/M_BenderTrail
graph {
 properties {
  bAutomaticallySetUsageInEditor=false
  bUsedWithNiagaraMeshParticles=true
  bUsedWithNiagaraRibbons=true
  bUsedWithNiagaraSprites=true
  blendMode=BLEND_Translucent
  shadingModel=MSM_Unlit
  usedShadingModels=Unlit
 }
 nodes [
  n1 = ScalarParameter() {defaultValue=1.15, parameterName=LightnessShiftAmount}
  MaterialExpressionVertexColor_1 = ParticleColor()
  n2 = SquareRoot(None=Abs(None=@MaterialExpressionVertexColor_1.RGB))
  n3 = TextureCoordinate()
  n4 = ComponentMask(None=@n3) {g=true}
  n5 = Multiply(A=@n4, B=OneMinus(None=@n4))
  n6 = ComponentMask(None=@n3) {r=true}
  n7 = OneMinus(None=@n6)
  n8 = Time()
  n9 = Constant() {r=1}
  n10 = TextureSample(UVs=Panner(Coordinate=TextureCoordinate() {coordinateIndex=1, uTiling=0.5, vTiling=0.25}, Time=Multiply(A=@n8, B=@n9)) {speedX=-2, speedY=0.15}) {texture=/DealershipSharedLibrary/FX/Textures/TexturePackage/T_VCVFX_FlameMask_RGB}
  n11 = TextureSample(UVs=Panner(Coordinate=TextureCoordinate() {coordinateIndex=1, uTiling=0.8, vTiling=0.5}) {speedX=-2.5, speedY=0.5}) {texture=&T_VCVFX_FlameMask_RGB}
  n12 = Multiply(A=@n5, B=Constant() {r=5})
  n13 = ComponentMask(None=Multiply(A=AppendVector(A=@n10.G, B=@n11.R), B=@n12)) {g=true}
  n14 = Add(A=Multiply(A=@n7, B=@n5), B=@n13)
  n15 = Multiply(A=Constant() {r=2}, B=@n7)
  MaterialExpressionConstantClamp_10 = Saturate(None=ConstantBiasScale(None=Multiply(A=@n14, B=@n15)) {bias=-0.25, scale=2})
  n16 = LinearInterpolate(
   A=Multiply(
    A=Constant() {r=2.5}
    B=MaterialFunctionCall!(Shift=@n1, Color=@n2) {materialFunction=/DealershipSharedLibrary/FX/Materials/MaterialFunctions/MF_VCVFX_LightnessShift}.Result
   )
   B=@n2
   Alpha=@MaterialExpressionConstantClamp_10
  )
  n17 = OneMinus(None=Multiply(A=@n6, B=@n13))
  MaterialExpressionConstantClamp_11 = Saturate(None=ConstantBiasScale(None=Multiply(A=@n17, B=@MaterialExpressionConstantClamp_10)) {bias=-0.3, scale=4})
  MaterialExpressionConstantClamp_9 = Saturate(None=Add(A=@MaterialExpressionConstantClamp_11, B=@n12))
  MaterialFunctionCall(Emissive=LinearInterpolate(A=@n2, B=Multiply(A=@n2, B=@n16), Alpha=@MaterialExpressionConstantClamp_9)) {materialFunction=&MF_FX_EmissiveAdjustments}
  n18 = Constant() {r=8}
  n19 = Multiply(A=@n3, B=@n12)
  MaterialExpressionConstantClamp_12 = Saturate(None=Multiply(A=@MaterialExpressionConstantClamp_10, B=Multiply(A=@n18, B=@n19)))
  Reroute(None=MaterialFunctionCall(Opacity In=@MaterialExpressionConstantClamp_12) {materialFunction=&MF_VCVFX_DepthFade}.Result)
  Transform(None=VectorParameter() {parameterName=WPO}.RGB) {transformSourceType=TRANSFORMSOURCE_Local}
 ]
}
