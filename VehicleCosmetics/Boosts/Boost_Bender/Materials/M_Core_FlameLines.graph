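# Unlit, masked, two-sided Niagara material for the Bender vehicle boost, drawing animated flame or bolt lines.
# Samples a flipbook texture normally and U-mirrored on a time-driven frame offset.
# Panned cloud noise on zero-tiled UVs gives hard-thresholded, time-varying switches that flicker between channels and mirroring.
# Particle color and dynamic Power shape the emissive adjustment input, while Hardness and particle alpha shape a separate mask value.
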Material /VehicleCosmetics/Boosts/Boost_Bender/Materials/M_Core_FlameLines
graph {
 properties {
  bAutomaticallySetUsageInEditor=false
  bUsedWithNiagaraMeshParticles=true
  bUsedWithNiagaraRibbons=true
  bUsedWithNiagaraSprites=true
  blendMode=BLEND_Masked
  shadingModel=MSM_Unlit
  twoSided=true
  usedShadingModels=Unlit
 }
 nodes [
  n1 = TextureCoordinate() {vTiling=0.9}
  n2 = ScalarParameter() {defaultValue=2, group=FlipBook, parameterName=Columns}
  n3 = ScalarParameter() {defaultValue=2, group=FlipBook, parameterName=Rows}
  n4 = AppendVector(A=@n2, B=@n3)
  n5 = AppendVector(A=@n3, B=Multiply(A=@n2, B=@n3))
  n6 = Frac(None=Multiply(A=Time(), B=ScalarParameter() {defaultValue=1, group=FlipBook, parameterName=Speed}))
  n7 = AppendVector(A=@n6, B=@n6)
  n8 = Divide(A=Floor(None=Multiply(A=@n5, B=@n7)), B=@n4)
  n9 = TextureSample(UVs=Add(A=Divide(A=@n1, B=@n4), B=@n8)) {texture=/VehicleCosmetics/Boosts/Boost_Bender/Textures/Bolt_Circle_FB_RGB_02_T}
  n10 = TextureCoordinate() {uTiling=0, vTiling=0}
  n11 = Panner(Coordinate=@n10) {speedX=1, speedY=0.65}
  n12 = Panner(Coordinate=@n10) {speedX=-1, speedY=-0.3}
  n13 = AppendVector(
   A=AppendVector(
    A=TextureSample(UVs=@n11) {texture=/VehicleCosmetics/FX/Textures/Noise/T_VCVFX_Clouds_A}.R
    B=TextureSample(UVs=@n12) {texture=&T_VCVFX_Clouds_A}.R
   )
   B=TextureSample(UVs=Panner(Coordinate=@n10) {speedX=-0.5, speedY=0.3}) {texture=&T_VCVFX_Clouds_A}.R
  )
  n14 = Constant() {r=0.2}
  MaterialExpressionConstantClamp_2 = Saturate(None=Multiply(A=Subtract(A=@n13, B=@n14), B=Constant() {r=35}))
  n15 = ComponentMask(None=@MaterialExpressionConstantClamp_2) {r=true}
  n16 = ComponentMask(None=@MaterialExpressionConstantClamp_2) {g=true}
  n17 = TextureCoordinate() {uTiling=-1, vTiling=0.9}
  n18 = TextureSample(UVs=Add(A=Divide(A=@n17, B=@n4), B=@n8)) {texture=&Bolt_Circle_FB_RGB_02_T}
  n19 = LinearInterpolate(
   A=LinearInterpolate(A=LinearInterpolate(A=@n9.R, B=@n9.R, Alpha=@n15), B=@n9.B, Alpha=@n16)
   B=LinearInterpolate(A=LinearInterpolate(A=@n18.G, B=@n18.G, Alpha=@n15), B=@n18.R, Alpha=@n16)
   Alpha=ComponentMask(None=@MaterialExpressionConstantClamp_2) {b=true}
  )
  MaterialExpressionConstantClamp_4 = Saturate(None=ConstantBiasScale(None=@n19) {bias=0, scale=8})
  MaterialExpressionConstantClamp_0 = Saturate(None=OneMinus(None=@MaterialExpressionConstantClamp_4))
  MaterialExpressionMeshEmitterVertexColor_0 = ParticleColor()
  MaterialExpressionMeshEmitterDynamicParameter_0 = DynamicParameter() {paramNames [Brightness,Power,HueShift,Hardness]}
  MaterialFunctionCall(Emissive=Multiply(
    A=@MaterialExpressionMeshEmitterVertexColor_0.RGB
    B=Power(Base=@MaterialExpressionConstantClamp_4, Exp=@MaterialExpressionMeshEmitterDynamicParameter_0.Power)
   )) {materialFunction=&MF_FX_EmissiveAdjustments}
  MaterialExpressionConstantClamp_1 = Saturate(None=ConstantBiasScale(None=@n19) {bias=0})
  MaterialExpressionConstantClamp_3 = Saturate(None=Multiply(A=@MaterialExpressionMeshEmitterDynamicParameter_0.Hardness, B=@MaterialExpressionConstantClamp_1))
  Reroute(None=Multiply(A=@MaterialExpressionConstantClamp_3, B=@MaterialExpressionMeshEmitterVertexColor_0.A))
 ]
}
