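# Unlit masked Niagara material that draws an animated fire shape for the Bender boost.
# Panning trail-mask noise, timed by particle relative time, distorts a smoke normal texture and radial dot-product falloffs.
# Smoke alpha, falloff and a sine-remapped noise term offset by particle alpha build an eroding mask.
# The combined texture also serves as a fake normal for Fresnel glow, tinted by particle color and fed through MF_FX_EmissiveAdjustments.
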Material /VehicleCosmetics/Boosts/Boost_Bender/Materials/M_FireConeShape
graph {
 properties {
  bAutomaticallySetUsageInEditor=false
  bUsedWithNiagaraMeshParticles=true
  bUsedWithNiagaraRibbons=true
  bUsedWithNiagaraSprites=true
  blendMode=BLEND_Masked
  shadingModel=MSM_Unlit
  usedShadingModels=Unlit
 }
 nodes [
  n1 = TextureCoordinate()
  n2 = ParticleRelativeTime()
  n3 = TextureSample(UVs=Panner(Coordinate=@n1, Time=@n2) {speedX=-0.1, speedY=0.2}) {texture=/DealershipSharedLibrary/FX/Textures/Noise/T_VCVFX_TrailMask_RGB}
  n4 = AppendVector(A=OneMinus(None=@n3.R), B=OneMinus(None=@n3.B))
  n5 = DynamicParameter() {paramNames [FresnelBrightness, FixPanner, UVDistort, FresnelAmount]}
  Add(A=@n4, B=@n5.FixPanner)
  n6 = TextureCoordinate()
  n7 = TextureSample(UVs=Panner(Coordinate=@n1, Time=@n2) {speedX=0.1, speedY=0.15}) {texture=&T_VCVFX_TrailMask_RGB}
  n8 = AppendVector(A=@n7.R, B=@n7.R)
  n9 = Multiply(A=@n8, B=@n4)
  n10 = Multiply(A=@n9, B=Constant() {r=0.05})
  n11 = ParticleRelativeTime()
  n12 = TextureSample(UVs=Panner(Coordinate=Add(A=@n6, B=@n10), Time=@n11) {speedX=0.01, speedY=0.01}) {texture=&T_VCVFX_TrailMask_RGB}
  MaterialExpressionConstantClamp_0 = Saturate(None=ConstantBiasScale(None=Desaturation(None=@n12.RGB)) {bias=0.4})
  MaterialExpressionVertexColor_1 = ParticleColor()
  n13 = OneMinus(None=Add(A=@MaterialExpressionVertexColor_1.A, B=Constant() {r=0.2}))
  MaterialExpressionConstantClamp_6 = Saturate(None=Sine(None=Add(A=@MaterialExpressionConstantClamp_0, B=@n13)) {period=2})
  n14 = Multiply(A=@n9, B=@n5.UVDistort)
  n15 = Add(A=@n14, B=Subtract(A=TextureCoordinate() {uTiling=3, vTiling=3}) {constB=1.75})
  n16 = DotProduct(A=@n15, B=@n15)
  MaterialExpressionConstantClamp_1 = Saturate(None=ConstantBiasScale(None=OneMinus(None=@n16)) {bias=0.5, scale=1})
  n17 = TextureCoordinate()
  n18 = TextureSample(UVs=Panner(Coordinate=Add(A=@n17, B=@n14), Time=@n11) {speedX=-0.04}) {samplerType=SAMPLERTYPE_Normal, texture=/VehicleCosmetics/Boosts/Boost_Bender/Textures/SmokeMask_02_N}
  MaterialExpressionConstantClamp_4 = Saturate(None=Multiply(A=@n18.A, B=@MaterialExpressionConstantClamp_1))
  MaterialExpressionConstantClamp_7 = Saturate(None=Multiply(A=@MaterialExpressionConstantClamp_4, B=@MaterialExpressionConstantClamp_6))
  n19 = Multiply(A=@MaterialExpressionVertexColor_1.RGB, B=@n5.FresnelBrightness)
  n20 = Multiply(A=@n18.RGB, B=Constant() {r=0.8})
  n21 = ComponentMask(None=AppendVector(A=@n8, B=@n4)) {b=true, g=true, r=true}
  n22 = DotProduct(A=@n15, B=@n15)
  MaterialExpressionConstantClamp_2 = Saturate(None=ConstantBiasScale(None=OneMinus(None=@n22)) {bias=-0.01, scale=2})
  MaterialExpressionConstantClamp_5 = Saturate(None=Add(A=Multiply(A=@n20, B=@MaterialExpressionConstantClamp_1), B=Multiply(A=@n21, B=@MaterialExpressionConstantClamp_2)))
  n23 = Multiply(A=@MaterialExpressionConstantClamp_5, B=@MaterialExpressionConstantClamp_1)
  n24 = Power(Base=Fresnel(Normal=@n23) {exponent=3}, Exp=@n5.FresnelAmount)
  n25 = Multiply(A=@MaterialExpressionVertexColor_1.RGB, B=Constant() {r=0.2})
  MaterialExpressionConstantClamp_3 = Saturate(None=Desaturation(None=@n23))
  n26 = Multiply(A=@MaterialExpressionConstantClamp_3, B=ScalarParameter() {defaultValue=10, parameterName=Brightness})
  MaterialFunctionCall(Emissive=Add(A=Multiply(A=@n19, B=@n24), B=Multiply(A=@n25, B=@n26))) {materialFunction=&MF_FX_EmissiveAdjustments}
 ]
}
